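; Gameplay ability for the ColdFern exotic weapon that launches frag grenades from the owning pawn on reload.
; Reload-start and ammo-consume delegates lead to DoOnce-gated spawn logic, commented as resetting on fire.
; CalculateGrenadesDirection fans grenades evenly around the pawn's facing, pitched upward, with scalable floats setting count, spacing and flight.
; A delayed timer explodes every stored grenade, then async AoE targeting applies a damage GameplayEffect.
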(event Ability|EventActivateAbility
  (bind _returnvalue (Ability|GetCurrentSourceWeapon))
  ; Check for on reload event, reset on fire (consume ammo)
  (EventDispatchers|AssignOnReloadStarted _returnvalue (delegate Custom|OnReloadStarted_Event))
  (EventDispatchers|AssignOnConsumeAmmo _returnvalue (delegate Custom|OnConsumeAmmo_Event)))

(event Custom|OnReloadStarted_Event
  (params (ReloadTime :type Float%20%28single-precision%29) (ReloadType :type EFortWeaponReloadType%20Enum :default "ReloadWholeClip"))
  ; Check for on reload event, reset on fire (consume ammo)
  (Utilities|FlowControl|DoOnce false
    (:Completed
      (bind _outfortplayerpawn (CallFunction|GetAvatarasValidFortPlayerPawn)
        (:ValidFortPlayerPawn
          (bind _output_get (Variables|Default|SetInstigatorActor _outfortplayerpawn))
          (Utilities|CrouchBlock _output_get)
          (if (ScalableFloat|GetValueasBool (Variables|Default|GetHF_ShouldSpawnGrenade?))
            (for _index (range (ScalableFloat|GetValueasInteger (Variables|Default|GetHF_NumberOfGrenades)))
              (bind (_spawn_location _spawn_direction) (CallFunction|CalculateGrenadesDirection _index (Variables|Default|GetInstigatorActor)))
              (bind _returnvalue (Projectile|SpawnProjectile :ProjectileClass (Variables|Default|GetGrenadeProjectile) :RequestedBy (Variables|Default|GetInstigatorActor) :SpawnLocation _spawn_location :SpawnRotation "0, 0, 0" :SpawnDirection _spawn_direction :bUseDefaultPhysics false :InitialSpeed (ScalableFloat|GetValueatLevel (Variables|Default|GetHF_GrenadeSpeed)) :GravityScale (ScalableFloat|GetValueatLevel (Variables|Default|GetHF_GrenadeGravity)) :ChargePercent 0.0))
              (bind _asb_prj_athena_frag_grenade (Utilities|Casting|CastToB_Prj_Athena_FragGrenade _returnvalue))
              (Utilities|Array|AddUnique (Variables|Default|GetSpawnedGrenades) _asb_prj_athena_frag_grenade)
              (Utilities|Time|SetTimerbyEvent (delegate Custom|ExplosionDamage) (ScalableFloat|GetValueatLevel (Variables|Default|GetHF_ExplosionDelay))))))))))

(event Custom|OnConsumeAmmo_Event
  (params (Quantity :type Integer))
  ; Check for on reload event, reset on fire (consume ammo)
  (Utilities|FlowControl|DoOnce false
    (:Completed
      (bind _outfortplayerpawn (CallFunction|GetAvatarasValidFortPlayerPawn)
        (:ValidFortPlayerPawn
          (bind _output_get (Variables|Default|SetInstigatorActor _outfortplayerpawn))
          (Utilities|CrouchBlock _output_get)
          (if (ScalableFloat|GetValueasBool (Variables|Default|GetHF_ShouldSpawnGrenade?))
            (for _index (range (ScalableFloat|GetValueasInteger (Variables|Default|GetHF_NumberOfGrenades)))
              (bind (_spawn_location _spawn_direction) (CallFunction|CalculateGrenadesDirection _index (Variables|Default|GetInstigatorActor)))
              (bind _returnvalue (Projectile|SpawnProjectile :ProjectileClass (Variables|Default|GetGrenadeProjectile) :RequestedBy (Variables|Default|GetInstigatorActor) :SpawnLocation _spawn_location :SpawnRotation "0, 0, 0" :SpawnDirection _spawn_direction :bUseDefaultPhysics false :InitialSpeed (ScalableFloat|GetValueatLevel (Variables|Default|GetHF_GrenadeSpeed)) :GravityScale (ScalableFloat|GetValueatLevel (Variables|Default|GetHF_GrenadeGravity)) :ChargePercent 0.0))
              (bind _asb_prj_athena_frag_grenade (Utilities|Casting|CastToB_Prj_Athena_FragGrenade _returnvalue))
              (Utilities|Array|AddUnique (Variables|Default|GetSpawnedGrenades) _asb_prj_athena_frag_grenade)
              (Utilities|Time|SetTimerbyEvent (delegate Custom|ExplosionDamage) (ScalableFloat|GetValueatLevel (Variables|Default|GetHF_ExplosionDelay))))))))))

(event Custom|ExplosionDamage
  ; Explosion and damage logic
  (for _array_element (Variables|Default|GetSpawnedGrenades)
    (Delivery|DoExplosion _array_element)
    (bind (_asynctaskproxy _targetinghandle) (Class|AsyncActionPerformTargeting|PerformTargetingAsyncAction _array_element "/WildEstateWeapons/Gameplay/ColdFern/Exotic/Explosion/TP_WildEstate_ColdFern_X_AoE.TP_WildEstate_ColdFern_X_AoE" true)
      (:Targeted
        (Ability|Tasks|ApplyGameplayEffectToHitResultsAsync (Ability|GetAbilitySystemComponent (Variables|Default|GetInstigatorActor)) (TargetingSystem|TargetingResults|GetTargetingResults (GameInstanceSubsystems|GetTargetingSubsystem) _targetinghandle) (Variables|Default|GetAoEGE) 0 1 (Class|AthenaFunctionLibrary|MakeSimpleFortEffectContext (Variables|Default|GetInstigatorActor) (Variables|Default|GetInstigatorActor)) (Variables|Default|GetHF_MaxTargets))))))

(fn CalculateGrenadesDirection
  (params (Index :type Integer) (InstigatorActor :type Actor%20Object%20Reference))
  (bind (_returnvalue_location _returnvalue_rotation) (Math|Transform|BreakTransform (Transformation|GetActorTransform InstigatorActor)))
  (return (+ _returnvalue_location "0.000000,0.000000,30.000000") (Math|Rotator|MakeRotator (.roll _returnvalue_rotation) (+ (.pitch _returnvalue_rotation) 1e+01) (+ (.yaw _returnvalue_rotation) (+ (* Index (ScalableFloat|GetValueatLevel (Variables|Default|GetHF_GrenadesRotationDistance))) (Utilities|Operators|Multiply (ScalableFloat|GetValueatLevel (Variables|Default|GetHF_GrenadesRotationDistance)) (/ (- (ScalableFloat|GetValueatLevel (Variables|Default|GetHF_NumberOfGrenades)) 1.0) 2.0) -1.0))))))
